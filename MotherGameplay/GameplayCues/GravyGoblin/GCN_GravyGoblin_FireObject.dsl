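; Burst gameplay cue for the GravyGoblin gravity gun that plays the sound for firing an object.
; On burst, it casts the cue target to a gravity gun holder and gets the owning player's pawn.
; It picks a first-person or standard sound cue based on whether that pawn is locally controlled, then spawns it attached to the holder's root component.
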(fn OnBurst(EitherVFX-preferthisversion)
  (params (MyTarget :type Actor%20Object%20Reference) (Parameters :type Gameplay%20Cue%20Parameters%20Structure%20%28by%20ref%29) (ParticleComponents :type Array%20of%20FXSystem%20Component%20Object%20References) (AudioComponents :type Array%20of%20Audio%20Component%20Object%20References) (BurstCameraShakeInstance :type Camera%20Shake%20Base%20Object%20Reference) (BurstDecalInstance :type Decal%20Actor%20Object%20Reference))
  (bind _asfort_gravity_gun_holder (Utilities|Casting|CastToFortGravityGunHolder MyTarget))
  (bind _asfort_pawn (Utilities|Casting|CastToFortPawn (Variables|/Script/FortniteGame.FortGravityGunHolder|GetOwningPlayer _asfort_gravity_gun_holder)))
  (Audio|SpawnSoundAttached (select (Pawn|IsLocallyControlled _asfort_pawn) "/MotherGameplay/Items/GravyGoblin/Assets/Audio/Launcher_Gravy_Goblin_Fire_Object_1P_Cue.Launcher_Gravy_Goblin_Fire_Object_1P_Cue" "/MotherGameplay/Items/GravyGoblin/Assets/Audio/Launcher_Gravy_Goblin_Fire_Object_Cue.Launcher_Gravy_Goblin_Fire_Object_Cue") (Variables|/Script/Engine.Actor|GetRootComponent _asfort_gravity_gun_holder))
  (return))
